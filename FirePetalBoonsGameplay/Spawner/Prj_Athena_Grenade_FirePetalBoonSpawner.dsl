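; Thrown grenade projectile Blueprint, by its name the Athena Fire Petal Boon Spawner, that drops boon item pickups where it lands.
; On hit it runs the parent's hit handling, then loops over its BoonsToSpawn list and spawns each entry as a single world pickup at the hit location, flagged as tossed.
; It then destroys itself.
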(event Collision|EventHit
  (params (MyComp :type Primitive%20Component%20Object%20Reference) (Other :type Actor%20Object%20Reference) (OtherComp :type Primitive%20Component%20Object%20Reference) (bSelfMoved :type Boolean) (HitLocation :type Vector :default "0, 0, 0") (HitNormal :type Vector :default "0, 0, 0") (NormalImpulse :type Vector :default "0, 0, 0") (Hit :type Hit%20Result%20Structure%20%28by%20ref%29))
  (|CallParentFunction MyComp Other OtherComp bSelfMoved HitLocation HitNormal NormalImpulse Hit)
  (for _array_element (Variables|Default|GetBoonsToSpawn)
    (Inventory|SpawnItemVariantPickupinWorld (Utilities|Struct|MakeSpawnItemVariantParams _array_element "(GameplayTags=)" 1 HitLocation "0.000000,0.000000,1.000000" 0 true true false -1 "Tossed" "Unset" 0 0 false)))
  (Actor|DestroyActor))

(fn ConstructionScript
  (|CallParentFunction))
